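# Sound cue for the Icicle Larva's air ram flight one-shot, layering a sprint whoosh over a boom.
# A mixer blends a randomly chosen, pitched-down and slightly delayed sprint one-shot with a louder, randomly chosen transform-end boom.
# It uses flight movement attenuation and the gadgets/items sound class, and sends to a flight sidechain volume audio bus.
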SoundCue /IcicleLarvaGameplay/Assets/Audio/Cues/IP_SC_IcicyleLarva_AirRam_Flight_Oneshot
root=Mixer_0
nodes [
 {id=Delay_0, children [Modulator_0], properties {delayMax=0.05, delayMin=0.05}}
 {id=Mixer_0, children [Delay_0,Modulator_1], properties {inputVolume [1.2,2]}}
 {id=Modulator_0, children [Random_0], properties {pitchMax=0.8, pitchMin=0.8}}
 {id=Modulator_1, children [Random_1]}
 {id=Random_0, children [WavePlayer_0,WavePlayer_1,WavePlayer_2,WavePlayer_3], properties {weights [1,1,1,1]}}
 {id=Random_1, children [WavePlayer_4,WavePlayer_5,WavePlayer_6], properties {weights [1,1,1]}}
 {id=WavePlayer_0, properties {soundWaveAssetPtr=/IcicleLarvaGameplay/Assets/Audio/Waves/IP_IcicleLarva_Sprint_OneShot_01}}
 {id=WavePlayer_1, properties {soundWaveAssetPtr=/IcicleLarvaGameplay/Assets/Audio/Waves/IP_IcicleLarva_Sprint_OneShot_02}}
 {id=WavePlayer_2, properties {soundWaveAssetPtr=/IcicleLarvaGameplay/Assets/Audio/Waves/IP_IcicleLarva_Sprint_OneShot_03}}
 {id=WavePlayer_3, properties {soundWaveAssetPtr=/IcicleLarvaGameplay/Assets/Audio/Waves/IP_IcicleLarva_Sprint_OneShot_04}}
 {id=WavePlayer_4, properties {soundWaveAssetPtr=/IcicleLarvaGameplay/Assets/Audio/Waves/IP_IcicleLarva_Transform_End_Boom_01}}
 {id=WavePlayer_5, properties {soundWaveAssetPtr=/IcicleLarvaGameplay/Assets/Audio/Waves/IP_IcicleLarva_Transform_End_Boom_02}}
 {id=WavePlayer_6, properties {soundWaveAssetPtr=/IcicleLarvaGameplay/Assets/Audio/Waves/IP_IcicleLarva_Transform_End_Boom_03}}
]
properties {
 attenuationSettings=/IcicleLarvaGameplay/Assets/Audio/Settings/ATT_IcicleLarva_Flight_Movement
 maxDistance=30500
 soundClassObject=&Fort_Gadgets_Items
 volumeMultiplier=1
 busSends [{audioBus=/IcicleLarvaGameplay/Assets/Audio/Modulation/AB_IcicleLarva_Flight_SideChain_Vol, minSendLevel=0, sendLevel=1, soundSourceBus=null}]
}
