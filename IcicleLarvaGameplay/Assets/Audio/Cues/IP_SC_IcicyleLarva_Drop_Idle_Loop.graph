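# Sound cue that plays the looping idle sound for the Icicle Larva drop, classed under gadget item sounds.
# A single looping wave player feeds the root directly, with slightly reduced volume and a dedicated idle attenuation asset.
# It also sends at full level to a flight side-chain volume audio bus, whose name suggests it drives volume ducking tied to the larva's flight audio.
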SoundCue /IcicleLarvaGameplay/Assets/Audio/Cues/IP_SC_IcicyleLarva_Drop_Idle_Loop
root=WavePlayer_4
nodes [{id=WavePlayer_4, properties {bLooping=true, soundWaveAssetPtr=/IcicleLarvaGameplay/Assets/Audio/Waves/IP_IcicleLarva_Drop_Idle}}]
properties {
 attenuationSettings=/IcicleLarvaGameplay/Assets/Audio/Settings/ATT_IcicleLarva_Idle
 maxDistance=2000
 soundClassObject=&Fort_Gadgets_Items
 volumeMultiplier=0.8
 busSends [{audioBus=/IcicleLarvaGameplay/Assets/Audio/Modulation/AB_IcicleLarva_Flight_SideChain_Vol, minSendLevel=0, sendLevel=1, soundSourceBus=null}]
}
