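# Sound cue for the Icicle Larva's one-shot flight takeoff sound, spatialized with the flight movement attenuation.
# A quality switch picks a random pool of all four takeoff waves for the first two levels, or just waves 01 and 04 for the other two.
# A root modulator holds volume steady, leaving pitch modulation at default.
# The cue also feeds a flight sidechain volume audio bus at full send.
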SoundCue /IcicleLarvaGameplay/Assets/Audio/Cues/IP_SC_IcicyleLarva_Flight_Takeoff_Oneshot
root=Modulator_0
nodes [
 {id=Modulator_0, children [QualityLevel_0], properties {volumeMax=1, volumeMin=1}}
 {id=QualityLevel_0, children [Random_1,Random_1,Random_0,Random_0]}
 {id=Random_0, children [WavePlayer_2,WavePlayer_5], properties {bShouldExcludeFromBranchCulling=true, weights [1,1]}}
 {id=Random_1, children [WavePlayer_2,WavePlayer_3,WavePlayer_4,WavePlayer_5], properties {weights [1,1,1,1]}}
 {id=WavePlayer_2, properties {soundWaveAssetPtr=/IcicleLarvaGameplay/Assets/Audio/Waves/IP_IcicleLarva_Flight_Takeoff_01}}
 {id=WavePlayer_3, properties {soundWaveAssetPtr=/IcicleLarvaGameplay/Assets/Audio/Waves/IP_IcicleLarva_Flight_Takeoff_02}}
 {id=WavePlayer_4, properties {soundWaveAssetPtr=/IcicleLarvaGameplay/Assets/Audio/Waves/IP_IcicleLarva_Flight_Takeoff_03}}
 {id=WavePlayer_5, properties {soundWaveAssetPtr=/IcicleLarvaGameplay/Assets/Audio/Waves/IP_IcicleLarva_Flight_Takeoff_04}}
]
properties {
 attenuationSettings=/IcicleLarvaGameplay/Assets/Audio/Settings/ATT_IcicleLarva_Flight_Movement
 maxDistance=30500
 soundClassObject=&Fort_Gadgets_Items
 volumeMultiplier=0.4
 busSends [{audioBus=/IcicleLarvaGameplay/Assets/Audio/Modulation/AB_IcicleLarva_Flight_SideChain_Vol, minSendLevel=0, sendLevel=1, soundSourceBus=null}]
}
