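# Sound cue for the Icicle Larva's one-shot sprint flight sound, classed under gadget and item audio.
# A modulator feeds from an equally weighted random pick among four sprint one-shot waves, with raised cue volume and pitch.
# It uses the larva's flight movement attenuation and sends to a flight sidechain volume audio bus.
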SoundCue /IcicleLarvaGameplay/Assets/Audio/Cues/IP_SC_IcicyleLarva_Sprint_Flight_Oneshot
root=Modulator_0
nodes [
 {id=Modulator_0, children [Random_0]}
 {id=Random_0, children [WavePlayer_0,WavePlayer_1,WavePlayer_2,WavePlayer_3], properties {weights [1,1,1,1]}}
 {id=WavePlayer_0, properties {soundWaveAssetPtr=/IcicleLarvaGameplay/Assets/Audio/Waves/IP_IcicleLarva_Sprint_OneShot_01}}
 {id=WavePlayer_1, properties {soundWaveAssetPtr=/IcicleLarvaGameplay/Assets/Audio/Waves/IP_IcicleLarva_Sprint_OneShot_02}}
 {id=WavePlayer_2, properties {soundWaveAssetPtr=/IcicleLarvaGameplay/Assets/Audio/Waves/IP_IcicleLarva_Sprint_OneShot_03}}
 {id=WavePlayer_3, properties {soundWaveAssetPtr=/IcicleLarvaGameplay/Assets/Audio/Waves/IP_IcicleLarva_Sprint_OneShot_04}}
]
properties {
 attenuationSettings=/IcicleLarvaGameplay/Assets/Audio/Settings/ATT_IcicleLarva_Flight_Movement
 maxDistance=30500
 pitchMultiplier=1.3
 soundClassObject=&Fort_Gadgets_Items
 volumeMultiplier=1.5
 busSends [{audioBus=/IcicleLarvaGameplay/Assets/Audio/Modulation/AB_IcicleLarva_Flight_SideChain_Vol, minSendLevel=0, sendLevel=1, soundSourceBus=null}]
}
